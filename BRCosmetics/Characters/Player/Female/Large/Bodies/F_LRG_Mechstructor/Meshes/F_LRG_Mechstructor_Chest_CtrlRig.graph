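# Control Rig for the Mechstructor large female body's chest mesh that procedurally adjusts the dyn_chest_01 bone as the right arm moves.
# It reads upperarm_r rotation relative to spine_05, remaps its Euler Z through clamped ranges, and applies the result as an additive local rotation.
# A second additive modify on that bone has an identity offset.
# A spine_01 bend calculation is present but feeds no bone.
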ControlRigBlueprint /BRCosmetics/Characters/Player/Female/Large/Bodies/F_LRG_Mechstructor/Meshes/F_LRG_Mechstructor_Chest_CtrlRig
graphs [
 RigVMFunctionLibrary {links [], nodes []}
 RigVMModel {nodes [
   RigUnit_BeginExecution = &RigUnit_BeginExecution()
   n1 = &RigVMFunction_MathQuaternionToEuler(RotationOrder=ZYX, Value=&RigUnit_GetRelativeBoneTransform(Bone=upperarm_r, Space=spine_05).Transform.Rotation)
   /Script/ControlRig.RigUnit_ModifyBoneTransforms(
    BoneToModify=((Bone="dyn_chest_01",Transform=(Rotation=(X=0,Y=0,Z=0,W=1),Translation=(X=0,Y=0,Z=0),Scale3D=(X=1,Y=1,Z=1))))
    Mode=AdditiveLocal
    Weight=1
    WeightMaximum=1
    WeightMinimum=0
    BoneToModify.0.Transform.Rotation=&RigVMFunction_MathQuaternionFromEuler(
     Euler=(X=0,Y=0,Z=0)
     Euler.X=&RigVMFunction_MathFloatRemap(SourceMaximum=126, SourceMinimum=18, TargetMaximum=-1, TargetMinimum=0, bClamp=true, Value=@n1.Result.Z).Result
     Euler.Y=&RigVMFunction_MathFloatRemap(SourceMaximum=126, SourceMinimum=18, TargetMaximum=25, TargetMinimum=0, bClamp=true, Value=@n1.Result.Z).Result
     Euler.Z=&RigVMFunction_MathFloatRemap(SourceMaximum=126, SourceMinimum=18, TargetMaximum=-5, TargetMinimum=0, bClamp=true, Value=@n1.Result.Z).Result
    ).Result
    ExecutePin=@RigUnit_BeginExecution.ExecutePin
   ) {is=struct RigVMUnitNode}
   > &RigUnit_ModifyBoneTransforms(
    BoneToModify=((Bone="dyn_chest_01",Transform=(Rotation=(X=0,Y=0,Z=0,W=1),Translation=(X=0,Y=0,Z=0),Scale3D=(X=1,Y=1,Z=1))))
    Mode=AdditiveLocal
    Weight=1
    WeightMaximum=1
    WeightMinimum=0
   )
   n2 = &RigVMFunction_MathQuaternionToEuler(RotationOrder=ZYX, Value=&RigUnit_GetRelativeBoneTransform(Bone=spine_01, Space=spine_05).Transform.Rotation)
   /Script/RigVM.RigVMFunction_MathFloatAbs(Value=@n2.Result.X) {is=struct RigVMUnitNode, template="Absolute::Execute(in Value,out Result)"}
   &RigVMFunction_MathFloatRemap(SourceMaximum=126, SourceMinimum=18, TargetMaximum=30, TargetMinimum=0, bClamp=true, Value=@n2.Result.Z)
   &RigVMFunction_MathFloatRemap(SourceMaximum=126, SourceMinimum=18, TargetMaximum=-10, TargetMinimum=0, bClamp=true, Value=@n2.Result.Z)
   &RigVMFunction_MathQuaternionFromEuler(
    Euler=(X=0,Y=0,Z=0)
    Euler.Z=&RigVMFunction_MathFloatRemap(SourceMaximum=0, SourceMinimum=-10, TargetMaximum=9, TargetMinimum=0, bClamp=true, Value=@n2.Result.X).Result
    Euler.X=&RigVMFunction_MathFloatRemap(SourceMaximum=0, SourceMinimum=-10, TargetMaximum=9, TargetMinimum=0, bClamp=true, Value=@n2.Result.X).Result
    Euler.Y=&RigVMFunction_MathFloatRemap(SourceMaximum=0, SourceMinimum=-10, TargetMaximum=9, TargetMinimum=0, bClamp=true, Value=@n2.Result.X).Result
   )
  ]}
]
hierarchy [
 root {loc=<0 0 0>, rot=<0 0 0 1>}
 pelvis {parent=root, loc=<0 -0.648 94.7699>, rot=<0.0127 0.707 -0.0127 -0.707>}
 spine_01 {parent=pelvis, loc=<5.0975 -0.6388 0>, rot=<0 0 0.034 -0.9994>}
 spine_02 {parent=spine_01, loc=<7.2518 -0.035 0>, rot=<0 0 0.0238 -0.9997>}
 spine_03 {parent=spine_02, loc=<6.0104 0 0>, rot=<0 0 0.0148 -0.9999>}
 spine_04 {parent=spine_03, loc=<8.1024 0 0>, rot=<0 0 -0.0132 -0.9999>}
 dyn_chest {parent=spine_04, loc=<0 0 0>, rot=<0.0584 -0.7047 -0.0584 -0.7047>}
 spine_05 {parent=spine_04, loc=<10.0146 0 0>, rot=<0 0 -0.0015 -1>}
 clavicle_l {parent=spine_05, loc=<4.1651 -0.3023 -5.0942>, rot=<0.0465 -0.7289 -0.009 -0.6829>}
 upperarm_l {parent=clavicle_l, loc=<11.2117 -1.1486 -0.0332>, rot=<0.0179 -0.2982 -0.028 -0.9539>}
 lowerarm_l {parent=upperarm_l, loc=<26.5013 -0.0835 0.2605>, rot=<0.0036 0.0161 -0.2298 -0.9731>}
 hand_l {parent=lowerarm_l, loc=<23.1356 -1.1606 -0.2612>, rot=<-0.6704 -0.0028 0.0567 -0.7398>}
 index_metacarpal_l {parent=hand_l, loc=<1.6535 0.0183 -2.96>, rot=<0.0158 -0.0124 -0.0173 -0.9996>}
 index_01_l {parent=index_metacarpal_l, loc=<8.3008 0.0358 -0.0524>, rot=<-0.0044 0.0485 0.0924 -0.9945>}
 index_02_l {parent=index_01_l, loc=<3.3587 -0.3004 0.0452>, rot=<-0.002 0.0027 0.3225 -0.9466>}
 index_03_l {parent=index_02_l, loc=<2.5578 -0.123 -0.0252>, rot=<-0.002 0.0067 0.2558 -0.9667>}
 middle_metacarpal_l {parent=hand_l, loc=<2.1856 -0.1725 -0.5984>, rot=<0.006 0.0168 -0.0329 -0.9993>}
 middle_01_l {parent=middle_metacarpal_l, loc=<7.3275 -0.2447 0.0129>, rot=<-0.0208 0.0042 0.0858 -0.9961>}
 middle_02_l {parent=middle_01_l, loc=<4.0123 0.1401 0.1175>, rot=<-0.0218 0.0258 0.3512 -0.9357>}
 middle_03_l {parent=middle_02_l, loc=<2.6077 0.0534 0.1373>, rot=<0.0082 0.0188 0.1708 -0.9851>}
 pinky_metacarpal_l {parent=hand_l, loc=<1.7616 -0.9753 3.3814>, rot=<-0.0397 0.0814 -0.069 -0.9935>}
 pinky_01_l {parent=pinky_metacarpal_l, loc=<7.1709 -0.4831 -0.1094>, rot=<-0.0654 -0.0035 0.2242 -0.9723>}
 pinky_02_l {parent=pinky_01_l, loc=<2.8274 0.0893 -0.1043>, rot=<0.0037 0.0011 0.2854 -0.9584>}
 pinky_03_l {parent=pinky_02_l, loc=<2.2046 0.0308 -0.0078>, rot=<-0.0002 0.0098 0.1445 -0.9895>}
 ring_metacarpal_l {parent=hand_l, loc=<2.2074 -0.3625 1.3638>, rot=<-0.0284 0.0722 -0.0458 -0.9959>}
 ring_01_l {parent=ring_metacarpal_l, loc=<7.0295 -0.6197 -0.1446>, rot=<-0.0489 -0.0252 0.1892 -0.9804>}
 ring_02_l {parent=ring_01_l, loc=<3.6174 0.4647 -0.1027>, rot=<0 0 0.2806 -0.9598>}
 ring_03_l {parent=ring_02_l, loc=<2.4152 0.1141 0.0199>, rot=<0 0 0.1189 -0.9929>}
 thumb_01_l {parent=hand_l, loc=<1.3118 -0.3191 -2.459>, rot=<0.6546 -0.4042 0.0664 -0.6354>}
 thumb_02_l {parent=thumb_01_l, loc=<5.9337 -0.0448 0.0401>, rot=<-0.0037 0.0213 0.298 -0.9543>}
 thumb_03_l {parent=thumb_02_l, loc=<2.2223 0.3682 -0.0907>, rot=<0.0118 -0.013 0.0066 -0.9998>}
 lowerarm_twist_01_l {parent=lowerarm_l, loc=<16.7573 0 0>, rot=<0 0 0 1>}
 lowerarm_twist_02_l {parent=lowerarm_l, loc=<11.5353 0 0>, rot=<0 0 0 1>}
 upperarm_twist_01_l {parent=upperarm_l, loc=<7.2135 -0.1134 0>, rot=<0 0 0 -1>}
 dyn_shoulder_pad {parent=upperarm_twist_01_l, loc=<0 0 0>, rot=<0.0019 0.5247 -0.0138 -0.8512>}
 upperarm_twist_02_l {parent=upperarm_l, loc=<19.7997 -0.1134 0>, rot=<0 0 0 -1>}
 clavicle_r {parent=spine_05, loc=<4.1652 -0.3023 5.0942>, rot=<0.7289 0.0465 -0.6829 0.009>}
 upperarm_r {parent=clavicle_r, loc=<-11.2117 1.1486 0.0332>, rot=<0.0179 -0.2982 -0.028 -0.9539>}
 lowerarm_r {parent=upperarm_r, loc=<-26.5014 0.0835 -0.2605>, rot=<0.0036 0.0161 -0.2298 -0.9731>}
 hand_r {parent=lowerarm_r, loc=<-23.1357 1.1606 0.2612>, rot=<-0.6704 -0.0028 0.0567 -0.7398>}
 index_metacarpal_r {parent=hand_r, loc=<-1.6534 -0.0181 2.9601>, rot=<0.0158 -0.0124 -0.0173 -0.9996>}
 index_01_r {parent=index_metacarpal_r, loc=<-8.3008 -0.0357 0.0524>, rot=<-0.0044 0.0485 0.0924 -0.9945>}
 index_02_r {parent=index_01_r, loc=<-3.3589 0.3003 -0.0453>, rot=<-0.002 0.0027 0.3225 -0.9466>}
 index_03_r {parent=index_02_r, loc=<-2.5578 0.123 0.0252>, rot=<-0.002 0.0067 0.2558 -0.9667>}
 middle_metacarpal_r {parent=hand_r, loc=<-2.1856 0.1726 0.5984>, rot=<0.006 0.0168 -0.0329 -0.9993>}
 middle_01_r {parent=middle_metacarpal_r, loc=<-7.3273 0.2447 -0.0129>, rot=<-0.0208 0.0042 0.0858 -0.9961>}
 middle_02_r {parent=middle_01_r, loc=<-4.0124 -0.14 -0.1175>, rot=<-0.0218 0.0258 0.3512 -0.9357>}
 middle_03_r {parent=middle_02_r, loc=<-2.6077 -0.0533 -0.1373>, rot=<0.0082 0.0188 0.1708 -0.9851>}
 pinky_metacarpal_r {parent=hand_r, loc=<-1.7617 0.9753 -3.3814>, rot=<-0.0397 0.0814 -0.069 -0.9935>}
 pinky_01_r {parent=pinky_metacarpal_r, loc=<-7.1707 0.4832 0.1094>, rot=<-0.0654 -0.0035 0.2242 -0.9723>}
 pinky_02_r {parent=pinky_01_r, loc=<-2.8274 -0.0892 0.1043>, rot=<0.0037 0.0011 0.2854 -0.9584>}
 pinky_03_r {parent=pinky_02_r, loc=<-2.2045 -0.031 0.0078>, rot=<-0.0002 0.0098 0.1445 -0.9895>}
 ring_metacarpal_r {parent=hand_r, loc=<-2.2073 0.3627 -1.3638>, rot=<-0.0284 0.0722 -0.0458 -0.9959>}
 ring_01_r {parent=ring_metacarpal_r, loc=<-7.0297 0.6197 0.1446>, rot=<-0.0489 -0.0252 0.1892 -0.9804>}
 ring_02_r {parent=ring_01_r, loc=<-3.6172 -0.4648 0.1026>, rot=<0 0 0.2806 -0.9598>}
 ring_03_r {parent=ring_02_r, loc=<-2.4152 -0.1141 -0.0199>, rot=<0 0 0.1189 -0.9929>}
 thumb_01_r {parent=hand_r, loc=<-1.3119 0.3192 2.459>, rot=<0.6546 -0.4042 0.0664 -0.6354>}
 thumb_02_r {parent=thumb_01_r, loc=<-5.9336 0.0447 -0.0401>, rot=<-0.0037 0.0213 0.298 -0.9543>}
 thumb_03_r {parent=thumb_02_r, loc=<-2.2223 -0.3682 0.0907>, rot=<0.0118 -0.013 0.0066 -0.9998>}
 lowerarm_twist_02_r {parent=lowerarm_r, loc=<-11.5354 0 0>, rot=<0 0 0 -1>}
 lowerarm_twist_01_r {parent=lowerarm_r, loc=<-16.7573 0 0>, rot=<-0.0106 0.0055 -0.0089 -0.9999>}
 upperarm_twist_01_r {parent=upperarm_r, loc=<-7.2135 0.1134 0>, rot=<0 0 0 -1>}
 upperarm_twist_02_r {parent=upperarm_r, loc=<-19.7998 0.1134 0>, rot=<0 0 0 -1>}
 neck_01 {parent=spine_05, loc=<11.3775 -0.6396 0>, rot=<0 0 -0.2003 -0.9797>}
 dyn_chest_01 {parent=spine_05, loc=<-2.2829 2.5416 0>, rot=<0.6506 -0.2769 0.2769 -0.6506>}
 dyn_neckline {parent=spine_05, loc=<4.3553 6.2343 0>, rot=<0 0 -0.4024 -0.9155>}
 thigh_l {parent=pelvis, loc=<-2.5624 -0.0642 -10.3834>, rot=<0.0004 0.0055 0.0155 -0.9999>}
 calf_l {parent=thigh_l, loc=<-39.2091 -0.1238 -0.6795>, rot=<0.0006 -0.0161 -0.0391 -0.9991>}
 calf_twist_02_l {parent=calf_l, loc=<-17.5927 -1.7943 -0.2638>, rot=<0.0237 0.0095 -0.0035 -0.9997>}
 foot_l {parent=calf_l, loc=<-43.624 -2.818 -0.713>, rot=<0.044 0.0061 0.0367 -0.9983>}
 ball_l {parent=foot_l, loc=<-4.6464 10.3094 -0.3137>, rot=<0.0001 -0.0001 -0.7063 -0.7079>}
 thigh_twist_01_l {parent=thigh_l, loc=<-23.0764 0 0>, rot=<0.0013 0 0 -1>}
 dyn_kneepad_l {parent=thigh_twist_01_l, loc=<-13.9006 5.4858 -1.1703>, rot=<0.1483 0.6878 0.701 0.1161>}
 dyn_hipTube_01 {parent=thigh_l, loc=<1.6108 10.813 -8.4507>, rot=<-0.0019 -0.711 0.0015 -0.7032>}
 thigh_r {parent=pelvis, loc=<-2.5623 -0.0642 10.3834>, rot=<0.0055 -0.0004 0.9999 0.0155>}
 calf_r {parent=thigh_r, loc=<39.2093 0.1238 0.6795>, rot=<0.0006 -0.0161 -0.0391 -0.9991>}
 calf_twist_02_r {parent=calf_r, loc=<17.5927 1.7943 0.2638>, rot=<0.0237 0.0095 -0.0035 -0.9997>}
 foot_r {parent=calf_r, loc=<43.6242 2.818 0.713>, rot=<0.044 0.0061 0.0367 -0.9983>}
 ball_r {parent=foot_r, loc=<4.6464 -10.3094 0.3137>, rot=<0.0001 -0.0001 -0.7063 -0.7079>}
 thigh_twist_01_r {parent=thigh_r, loc=<23.0764 0 0>, rot=<0.0013 0 0 -1>}
 dyn_kneepad_r {parent=thigh_twist_01_r, loc=<13.9007 -5.4858 1.1703>, rot=<0.0726 0 0 -0.9974>}
 dyn_legstrap_r {parent=thigh_twist_01_r, loc=<-3.3057 -10.2025 0.9326>, rot=<-0.351 0.6155 0.6267 -0.3242>}
 dyn_waistBox {parent=pelvis, loc=<10.0758 -14.106 0>, rot=<0.6239 -0.3329 0.3329 -0.6239>}
 dyn_hipTube_mstr {parent=pelvis, loc=<11.4802 -2.2033 -6.5966>, rot=<-0.0127 -0.707 0.0127 -0.707>}
 dyn_hipTube_02 {parent=dyn_hipTube_mstr, loc=<16.7897 -9.6984 -7.7071>, rot=<0 0 0 1>}
]
