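# Physics asset for the large female Mechstructor player body, covering only the hip tube and pelvis for what appears to be secondary motion of the tube.
# A capsule body on the dyn_hipTube_mstr bone pairs with a box body on the pelvis.
# One constraint joins the tube to the pelvis, allowing asymmetric limited swing with almost no twist.
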PhysicsAsset /BRCosmetics/Characters/Player/Female/Large/Bodies/F_LRG_Mechstructor/Meshes/F_LRG_Mechstructor_Physics
bodies {
 dyn_hipTube_mstr [{shapeType=Capsule, length=12.2862, radius=6.5802, center=<8.3138 -4.2532 -3.8556>, rotation=<rot 64.2189 5.7507 30.0874>}]
 pelvis [{shapeType=Box, extentX=23.9175, extentY=9, extentZ=41.1622, center=<1.377 -2.3796 -2.8597>, rotation=<rot 8.1817 -89.9999 31.1732>}]
}
constraints [{bone1Name=dyn_hipTube_mstr, bone2Name=pelvis, swing1LimitDegrees=20, swing2LimitDegrees=45, twistLimitDegrees=1}]
