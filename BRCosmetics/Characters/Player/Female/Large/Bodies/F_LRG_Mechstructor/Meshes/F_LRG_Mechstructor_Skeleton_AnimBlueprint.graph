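# Character part AnimBlueprint for the large female Mechstructor body, adding secondary motion and pose corrections to its gear bones.
# It copies the parent mesh pose, then selects an emote/front-end Control Rig or chest-rig and bone-offset corrections by aim and crouch state.
# Weighted constraints, AnimDynamics and RigidBody then drive the hip tubes, kneepads, waist box and leg strap.
# The EventGraph derives pistol, rifle and rocket aim flags from the weapon core animation each update.
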AnimBlueprint /BRCosmetics/Characters/Player/Female/Large/Bodies/F_LRG_Mechstructor/Meshes/F_LRG_Mechstructor_Skeleton_AnimBlueprint
target_skeleton=&/F_LRG_Mechstructor_Skeleton
graphs [
 AnimGraph {nodes [
   n1 = "Transform (Modify) Bone - Bone: dyn_chest_01"(ComponentPose=Local To Component(), Alpha=1, Translation="-2,1,0", Rotation="0,0,0", Scale="1,1,1") {
    rotationMode=BMM_Additive
    scaleMode=BMM_Additive
    translationMode=BMM_Additive
    alphaBoolBlend {blendOption=Linear}
    boneToModify {boneName=dyn_chest_01}
    translation=<-2 1>
   }
   Component To Local(ComponentPose=@n1)
   n2 = "Transform (Modify) Bone - Bone: dyn_hipTube_01"(ComponentPose=Local To Component(LocalPose=Use cached pose 'AnimCache'()), Alpha=1, Translation="-10,8,25", Rotation="0,0,75", Scale="1,1,1") {
    rotationMode=BMM_Additive
    scaleMode=BMM_Additive
    translationMode=BMM_Additive
    alphaBoolBlend {blendOption=Linear}
    boneToModify {boneName=dyn_hipTube_01}
    rotation {roll=75}
    translation=<-10 8 25>
   }
   n3 = "Transform (Modify) Bone - Bone: dyn_hipTube_02"(ComponentPose=@n2, Alpha=1, Translation="-5,-10,0", Rotation="-25,0.000,0", Scale="1,1,1") {
    rotationMode=BMM_Additive
    scaleMode=BMM_Additive
    translationMode=BMM_Additive
    alphaBoolBlend {blendOption=Linear}
    boneToModify {boneName=dyn_hipTube_02}
    rotation {pitch=-25}
    translation=<-5 -10>
   }
   n4 = Blend Poses by bool(BlendPose_0=Component To Local(ComponentPose=@n3), BlendPose_1=Use cached pose 'AnimCache'(), bActiveValue=Get bIsCrouching()) {blendType=Linear}
   n5 = Control Rig(Source=Use cached pose 'AnimCache'()) {
    alphaBoolBlend {blendOption=Linear}
    controlRigAssetReference {blueprintRigClass=&/F_LRG_Mechstructor_Chest_CtrlRig.F_LRG_Mechstructor_Chest_CtrlRig_C}
   }
   n6 = "Transform (Modify) Bone - Bone: dyn_hipTube_01"(ComponentPose=Local To Component(LocalPose=@n5), Alpha=1, Translation="-10,8,25", Rotation="0,0,75", Scale="1,1,1") {
    rotationMode=BMM_Additive
    scaleMode=BMM_Additive
    translationMode=BMM_Additive
    alphaBoolBlend {blendOption=Linear}
    boneToModify {boneName=dyn_hipTube_01}
    rotation {roll=75}
    translation=<-10 8 25>
   }
   n7 = "Transform (Modify) Bone - Bone: dyn_hipTube_02"(ComponentPose=@n6, Alpha=1, Translation="-5,-10,0", Rotation="-25,0.000,0", Scale="1,1,1") {
    rotationMode=BMM_Additive
    scaleMode=BMM_Additive
    translationMode=BMM_Additive
    alphaBoolBlend {blendOption=Linear}
    boneToModify {boneName=dyn_hipTube_02}
    rotation {pitch=-25}
    translation=<-5 -10>
   }
   n8 = "Transform (Modify) Bone - Bone: dyn_chest_01"(ComponentPose=@n7, Alpha=1, Translation="-1.3,1.2,-1", Rotation="0,-5,0", Scale="1,1,1") {
    rotationMode=BMM_Additive
    scaleMode=BMM_Additive
    translationMode=BMM_Additive
    alphaBoolBlend {blendOption=Linear}
    boneToModify {boneName=dyn_chest_01}
    rotation {yaw=-5}
    translation=<-1.3 1.2 -1>
   }
   n9 = Component To Local(ComponentPose=@n8)
   n10 = AND Boolean(A=Get bIsTargeting(), B=Get Is Aiming Pistol() -> (Is Aiming Pistol:bool))
   n11 = Control Rig(Source=Use cached pose 'AnimCache'()) {
    alphaBoolBlend {blendOption=Linear}
    controlRigAssetReference {blueprintRigClass=&F_LRG_Mechstructor_Chest_CtrlRig_C}
   }
   n12 = "Transform (Modify) Bone - Bone: dyn_chest_01"(ComponentPose=Local To Component(LocalPose=@n11), Alpha=1, Translation="-0.2,0,0", Rotation="0,-4,0", Scale="1,1,1") {
    rotationMode=BMM_Additive
    scaleMode=BMM_Additive
    translationMode=BMM_Additive
    alphaBoolBlend {blendOption=Linear}
    boneToModify {boneName=dyn_chest_01}
    rotation {yaw=-4}
    translation {x=-0.2}
   }
   n13 = Get Is Aiming Rocket() -> (Is Aiming Rocket:bool)
   n14 = Control Rig(Source=Use cached pose 'AnimCache'()) {
    alphaBoolBlend {blendOption=Linear}
    controlRigAssetReference {blueprintRigClass=&F_LRG_Mechstructor_Chest_CtrlRig_C}
   }
   n15 = "Transform (Modify) Bone - Bone: dyn_chest_01"(ComponentPose=Local To Component(LocalPose=@n14), Alpha=1, Translation="-1.8,1,0", Rotation="0,-5,0", Scale="1,1,1") {
    rotationMode=BMM_Additive
    scaleMode=BMM_Additive
    translationMode=BMM_Additive
    alphaBoolBlend {blendOption=Linear}
    boneToModify {boneName=dyn_chest_01}
    rotation {yaw=-5}
    translation=<-1.8 1>
   }
   n16 = Blend Poses by bool(
    BlendPose_0=Component To Local(ComponentPose=@n15)
    BlendPose_1=Blend Poses by bool(
     BlendPose_0=Component To Local(ComponentPose=@n12)
     BlendPose_1=Blend Poses by bool(BlendPose_0=@n9, BlendPose_1=@n4, BlendTime_1=0.3, bActiveValue=@n10) {blendType=Linear}
     BlendTime_0=0.01
     BlendTime_1=0.3
     bActiveValue=@n13
    ) {blendType=Linear}
    BlendTime_0=0.01
    BlendTime_1=0.3
    bActiveValue=AND Boolean(A=Get bIsTargeting(), B=Get Is Aiming Rifle() -> (Is Aiming Rifle:bool))
   ) {blendType=Linear}
   n17 = Control Rig(Source=Use cached pose 'AnimCache'()) {
    alphaBoolBlend {blendOption=Linear}
    controlRigAssetReference {blueprintRigClass=&/F_LRG_Mechstructor_emote_CtrlRig.F_LRG_Mechstructor_emote_CtrlRig_C}
   }
   n18 = OR Boolean(A=Get bIsPlayingEmote(), B=Get bIsInFrontEnd())
   n19 = SelectFloat(A=0.45, B=1, bPickA=Get bIsCrouching())
   n20 = "Constraint - dyn_hipTube_01 (PP)"!(
    ComponentPose=Local To Component(LocalPose=Blend Poses by bool(BlendPose_0=@n17, BlendPose_1=@n16, BlendTime_0=0.01, BlendTime_1=0.3, bActiveValue=@n18) {blendType=Linear})
    Alpha=@n19
    ConstraintWeights_0=0.8
    ConstraintWeights_1:float=0.2
   ) -> (Pose) {
    alphaBoolBlend {blendOption=Linear}
    boneToModify {boneName=dyn_hipTube_01}
    constraintSetup [{targetBone {boneName=thigh_l}}, {targetBone {boneName=pelvis}}]
    constraintWeights [1,1]
   }
   n21 = "Constraint - dyn_hipTube_02 (PP)"!(ComponentPose=@n20, Alpha=@n19, ConstraintWeights_0=0.2, ConstraintWeights_1:float=0.8) -> (Pose) {
    alphaBoolBlend {blendOption=Linear}
    boneToModify {boneName=dyn_hipTube_02}
    constraintSetup [{targetBone {boneName=thigh_l}}, {targetBone {boneName=pelvis}}]
    constraintWeights [1,1]
   }
   n22 = "Constraint - dyn_kneepad_r (PPPP)"!(
    ComponentPose=@n21
    Alpha=1
    ConstraintWeights_0=0.4549
    ConstraintWeights_1:float=0.0019
    ConstraintWeights_2:float=0.0126
    ConstraintWeights_3:float=0.5306
   ) -> (Pose) {
    alphaBoolBlend {blendOption=Linear}
    boneToModify {boneName=dyn_kneepad_r}
    constraintSetup [
     {targetBone {boneName=thigh_r}}
     {targetBone {boneName=calf_r}}
     {targetBone {boneName=calf_twist_02_r}}
     {targetBone {boneName=thigh_twist_01_r}}
    ]
    constraintWeights [1,1,1,1]
   }
   n23 = "Constraint - dyn_kneepad_l (PPPP)"!(
    ComponentPose=@n22
    Alpha=1
    ConstraintWeights_0=0.379
    ConstraintWeights_1:float=0.0095
    ConstraintWeights_2:float=0.6107
    ConstraintWeights_3:float=0.0008
   ) -> (Pose) {
    alphaBoolBlend {blendOption=Linear}
    boneToModify {boneName=dyn_kneepad_l}
    constraintSetup [
     {targetBone {boneName=thigh_l}}
     {targetBone {boneName=calf_l}}
     {targetBone {boneName=thigh_twist_01_l}}
     {targetBone {boneName=calf_twist_02_l}}
    ]
    constraintWeights [1,1,1,1]
   }
   n24 = "Transform (Modify) Bone - Bone: dyn_hipTube_01"(ComponentPose=@n23, Alpha=1, Translation="0.000,0.000,3", Rotation="0,0,0", Scale="1,1,1") {
    rotationMode=BMM_Additive
    scaleMode=BMM_Additive
    translationMode=BMM_Additive
    alphaBoolBlend {blendOption=Linear}
    boneToModify {boneName=dyn_hipTube_01}
    translation {z=3}
   }
   n25 = "Transform (Modify) Bone - Bone: dyn_hipTube_02"(ComponentPose=@n24, Alpha=1, Translation="-2.630,-0.000,4", Rotation="-17.726,-6.240,1.528", Scale="1,1,1") {
    rotationMode=BMM_Additive
    scaleMode=BMM_Additive
    translationMode=BMM_Additive
    alphaBoolBlend {blendOption=Linear}
    boneToModify {boneName=dyn_hipTube_02}
    rotation=<rot -17.726 -6.24 1.528>
    translation {x=-2.63, z=4}
   }
   n26 = Anim Dynamics - Bone: dyn_hipTube_01(ComponentPose=@n25, Alpha=SelectFloat(A=0.05, B=0.2, bPickA=Get bIsCrouching()), ExternalForce="0,0,0") {
    angularDampingOverride=0.6
    angularSpringConstant=100
    bGravityOverrideInSimSpace=true
    bLinearSpring=true
    bOverrideAngularDamping=true
    bOverrideLinearDamping=true
    linearDampingOverride=0.97
    linearSpringConstant=250
    simulationSpace=BoneRelative
    alphaBoolBlend {blendOption=Linear}
    boundBone {boneName=dyn_hipTube_01}
    physicsBodyDefinitions [{
      sphereCollisionRadius=0
      boundBone {boneName=dyn_hipTube_01}
      boxExtents=<10 10 5>
      constraintSetup {angularLimitsMax=<2 2 2>, angularLimitsMin=<-2 -2 -2>, angularTarget=<0 0 0>, linearAxesMax=<5 5 10>, linearAxesMin=<-5 -5 -10>}
      localJointOffset=<0 0 -3>
     }]
    relativeSpaceBone {boneName=root}
   }
   n27 = Anim Dynamics - Bone: dyn_hipTube_02(ComponentPose=@n26, Alpha=SelectFloat(A=0.15, B=0.2, bPickA=Get bIsCrouching()), ExternalForce="0,0,0") {
    angularDampingOverride=0.6
    angularSpringConstant=100
    bGravityOverrideInSimSpace=true
    bLinearSpring=true
    bOverrideAngularDamping=true
    bOverrideLinearDamping=true
    linearDampingOverride=0.97
    linearSpringConstant=250
    simulationSpace=BoneRelative
    alphaBoolBlend {blendOption=Linear}
    boundBone {boneName=dyn_hipTube_02}
    physicsBodyDefinitions [{
      sphereCollisionRadius=0
      boundBone {boneName=dyn_hipTube_02}
      boxExtents=<10 10 5>
      constraintSetup {angularLimitsMax=<0 0 0>, angularLimitsMin=<0 0 0>, angularTarget=<0 0 0>, linearAxesMax=<2.5 2.5 10>, linearAxesMin=<-2.5 -2.5 -10>}
      localJointOffset=<0 0 -3>
     }]
    relativeSpaceBone {boneName=root}
   }
   n28 = RigidBody(ComponentPose=@n27) {
    bClampLinearTranslationLimitToRefPose=true
    bOverrideWorldGravity=true
    overridePhysicsAsset=&/F_LRG_Mechstructor_Physics
    simulationSpace=BaseBoneSpace
    alphaBoolBlend {blendOption=Linear}
    baseBoneRef {boneName=root}
    componentLinearAccScale=<0.35 0.35 0.4>
    componentLinearVelScale=<0.15 0.15 0.2>
    overrideWorldGravity {z=-980}
   }
   n29 = "Constraint - dyn_waistBox (PPPPPPP)"!(
    ComponentPose=@n28
    Alpha=1
    ConstraintWeights_0=0.0043
    ConstraintWeights_1:float=0.1579
    ConstraintWeights_2:float=0.0541
    ConstraintWeights_3:float=0.0764
    ConstraintWeights_4:float=0.0045
    ConstraintWeights_5:float=0.7016
    ConstraintWeights_6:float=0.0014
   ) -> (Pose) {
    alphaBoolBlend {blendOption=Linear}
    boneToModify {boneName=dyn_waistBox}
    constraintSetup [
     {targetBone {boneName=thigh_twist_01_r}}
     {targetBone {boneName=spine_01}}
     {targetBone {boneName=spine_02}}
     {targetBone {boneName=spine_03}}
     {targetBone {boneName=spine_04}}
     {targetBone {boneName=pelvis}}
     {targetBone {boneName=thigh_twist_01_l}}
    ]
    constraintWeights [1,1,1,1,1,1,1]
   }
   n30 = Anim Dynamics - Bone: dyn_waistBox(ComponentPose=@n29, Alpha=0.1, ExternalForce="0,0,0") {
    angularDampingOverride=0.9999
    angularSpringConstant=250
    bAngularSpring=true
    bGravityOverrideInSimSpace=true
    bLinearSpring=true
    bOverrideAngularDamping=true
    bOverrideLinearDamping=true
    gravityScale=0.35
    lODThreshold=1
    linearDampingOverride=0.9999
    linearSpringConstant=250
    simulationSpace=RootRelative
    alphaBoolBlend {blendOption=Linear}
    boundBone {boneName=dyn_waistBox}
    componentAppliedLinearAccClamp=<5000 5000 5000>
    physicsBodyDefinitions [{
      sphereCollisionRadius=0
      boundBone {boneName=dyn_waistBox}
      boxExtents=<10 10 5>
      constraintSetup {
       angularTargetAxis=AxisZ
       angularLimitsMax=<45 45 45>
       angularLimitsMin=<-45 -45 -45>
       angularTarget=<0 0 1>
       linearAxesMax=<1 1 5>
       linearAxesMin=<-1 -1 -5>
      }
      localJointOffset=<0 0 0>
     }]
    relativeSpaceBone {boneName=root}
   }
   n31 = "Constraint - dyn_kneepad_l (PP)"!(ComponentPose=@n30, Alpha=1, ConstraintWeights_0=1, ConstraintWeights_1:float=0) -> (Pose) {
    alphaBoolBlend {blendOption=Linear}
    boneToModify {boneName=dyn_kneepad_l}
    constraintSetup [{targetBone {boneName=thigh_twist_01_l}}, {targetBone {boneName=calf_l}}]
    constraintWeights [1,1]
   }
   n32 = "Constraint - dyn_kneepad_r (PP)"!(ComponentPose=@n31, Alpha=1, ConstraintWeights_0=1, ConstraintWeights_1:float=0.8) -> (Pose) {
    alphaBoolBlend {blendOption=Linear}
    boneToModify {boneName=dyn_kneepad_r}
    constraintSetup [{targetBone {boneName=thigh_twist_01_r}}, {targetBone {boneName=calf_r}}]
    constraintWeights [1,1]
   }
   n33 = "Transform (Modify) Bone - Bone: dyn_kneepad_r"(ComponentPose=@n32, Alpha="To Float (Boolean)"(InBool=Get bIsCrouching()), Translation="0,-2,2", Rotation="0,0,0", Scale="1,1,1") {
    rotationMode=BMM_Additive
    scaleMode=BMM_Additive
    translationMode=BMM_Additive
    alphaBoolBlend {blendOption=Linear}
    boneToModify {boneName=dyn_kneepad_r}
    translation {y=-2, z=2}
   }
   n34 = Anim Dynamics - Bone: dyn_legstrap_r(ComponentPose=@n33, Alpha=1, ExternalForce="0,0,0") {
    angularDampingOverride=0.5
    angularSpringConstant=25
    bAngularSpring=true
    bGravityOverrideInSimSpace=true
    bOverrideAngularDamping=true
    bOverrideLinearDamping=true
    gravityScale=0.2
    linearSpringConstant=250
    simulationSpace=BoneRelative
    alphaBoolBlend {blendOption=Linear}
    boundBone {boneName=dyn_legstrap_r}
    physicsBodyDefinitions [{
      sphereCollisionRadius=0
      boundBone {boneName=dyn_legstrap_r}
      boxExtents=<10 10 5>
      constraintSetup {angularLimitsMax=<25 25 25>, angularLimitsMin=<-25 -20 -25>, angularTarget=<0 0 0>}
      localJointOffset=<0 0 -3>
     }]
    relativeSpaceBone {boneName=root}
   }
   Output Pose(Result=Component To Local(ComponentPose=@n34)) {layerGroup=None}
   Save cached pose 'AnimCache'(Pose=Copy Pose From Mesh() {bCopyCurves=true, bUseAttachedParent=true}) {cacheName=AnimCache}
   AND Boolean(A=Get bIsTargeting(), B=@n13)
  ]}
 EventGraph {nodes [
   Event BlueprintUpdateAnimation()
   > Set Is Aiming Pistol(Is Aiming Pistol:bool=false) -> (Output_Get:bool)
   > Set Is Aiming Rocket(Is Aiming Rocket:bool=false) -> (Output_Get:bool)
   > Set Is Aiming Rifle(Is Aiming Rifle:bool=false) -> (Output_Get:bool)
   > n1 = Switch on EFortWeaponCoreAnimation!(
    execute
    Selection:EFortWeaponCoreAnimation Enum=GetWeaponCoreAnim!(self:Fort Player Animation Interface Object Reference=GetOwnerAnimationInterface!(self:Custom Character Part Anim Instance Object Reference) -> (ReturnValue:Fort Player Animation Interface Object Reference)) -> (ReturnValue:EFortWeaponCoreAnimation Enum)
   ) -> (
    Melee:Exec
    Pistol:Exec
    Shotgun:Exec
    PaperBlueprint:Exec
    Rifle:Exec
    MeleeOneHand:Exec
    MachinePistol:Exec
    RocketLauncher:Exec
    GrenadeLauncher:Exec
    GoingCommando:Exec
    AssaultRifle:Exec
    TacticalShotgun:Exec
    SniperRifle:Exec
    TrapPlacement:Exec
    ShoulderLauncher:Exec
    AbilityDecoTool:Exec
    Crossbow:Exec
    C4:Exec
    RemoteControl:Exec
    DualWield:Exec
    AR_BullPup:Exec
    AR_ForwardGrip:Exec
    MedPackPaddles:Exec
    SMG_P90:Exec
    AR_DrumGun:Exec
    Consumable_Small:Exec
    Consumable_Large:Exec
    Balloon:Exec
    MountedTurret:Exec
    CreativeTool:Exec
    ExplosiveBow:Exec
    AshtonIndigo:Exec
    AshtonChicago:Exec
    MeleeDualWield:Exec
    Unarmed:Exec
   )
   Branch(execute=[@n1.AssaultRifle, @n1.Crossbow, @n1.GrenadeLauncher, @n1.Rifle, @n1.Shotgun, @n1.SniperRifle, @n1.TacticalShotgun])
   > Set Is Aiming Rifle(Is Aiming Rifle=true)
   Branch(execute=[@n1.MachinePistol, @n1.Pistol])
   > Set Is Aiming Pistol(Is Aiming Pistol=true)
   Branch(execute=@n1.RocketLauncher)
   > Set Is Aiming Rocket(Is Aiming Rocket=true)
   TryGetPawnOwner()
  ]}
]
variables [
 Is Aiming Pistol {type=boolean}
 Is Aiming Rifle {type=boolean}
 Is Aiming Rocket {type=boolean}
]
