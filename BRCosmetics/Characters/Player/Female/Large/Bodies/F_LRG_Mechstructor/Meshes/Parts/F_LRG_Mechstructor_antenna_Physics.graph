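# Physics asset for the antenna part of the large female Mechstructor player character body.
# A zero-length head capsule acts as a sphere, with two capsule bodies on the dyn_antenna bones forming a chain from it.
# The antenna base joint allows limited, asymmetric swing against the head with twist nearly locked.
# The tip joint is almost fully locked to the base segment, so the antenna moves as a nearly rigid piece.
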PhysicsAsset /BRCosmetics/Characters/Player/Female/Large/Bodies/F_LRG_Mechstructor/Meshes/Parts/F_LRG_Mechstructor_antenna_Physics
bodies {
 dyn_antenna_01 [{shapeType=Capsule, length=11.505, radius=0.505, center=<0.0633 -0.0433 6.1074>, rotation=<rot -0.174 -90 0.5207>}]
 dyn_antenna_02 [{shapeType=Capsule, length=1.3699, radius=0.7153, center=<-0.0212 0.1037 1.432>, rotation=<rot 0.7966 -90 -0.366>}]
 head [{shapeType=Capsule, length=0, radius=7.2121, center=<6.6861 -1.4616 0.0342>, rotation=<rot 1.4623 -90 -2.5886>}]
}
constraints [
 {bone1Name=dyn_antenna_01, bone2Name=head, swing1LimitDegrees=10, swing2LimitDegrees=2, twistLimitDegrees=0.001}
 {bone1Name=dyn_antenna_02, bone2Name=dyn_antenna_01, swing1LimitDegrees=0.02, swing2LimitDegrees=0.02, twistLimitDegrees=0.02}
]
