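# Physics asset for the Mechstructor female character's hair and headset part, providing bodies and joints for secondary hair motion.
# Capsules on upper spine, clavicles, neck and head approximate the body as collision volumes, linked by angularly free constraints.
# Mostly box-shaped hair bodies form front, side and back chains rooted at the head, with the right side branching.
# Hair joints use limited swing and very tight twist, keeping strands from rolling.
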PhysicsAsset /BRCosmetics/Characters/Player/Female/Large/Bodies/F_LRG_Mechstructor/Meshes/Parts/F_MED_Mechstructor_Hair_headset_FaceAcc_Physics
bodies {
 clavicle_l [{shapeType=Capsule, length=11.2128, radius=7.9177, center=<3.8385 -3.1846 -1.7113>, rotation=<rot 0 -89.9993 106.1269>}]
 clavicle_r [
  {shapeType=Capsule, length=12.7889, radius=7.5916, center=<-2.969 2.8928 0.991>, rotation=<rot -1.1448 -85.0614 103.0219>}
  {shapeType=Capsule, length=2.0389, radius=2.8416, center=<-2.8497 7.0414 -2.162>, rotation=<rot -1.1448 -85.0613 103.022>}
  {shapeType=Capsule, length=15.7889, radius=4.3416, center=<-6.127 -9.5298 -2.8673>, rotation=<rot 11.9853 -14.7907 89.0119>}
 ]
 dyn_hair_bk_01 [{shapeType=Box, extentX=6.1017, extentY=7.5, extentZ=1.7408, center=<3.3857 0 0>}]
 dyn_hair_bk_02 [{shapeType=Box, extentX=6.1017, extentY=12.3439, extentZ=2.1269, center=<3.3857 0.0867 0.642>}]
 dyn_hair_bk_03 [{shapeType=Capsule, length=9.5, radius=1.75, shapeName=dyn_hair_bk_04_capsule, center=<2.4891 0 0.5593>, rotation=<rot 0 0 90>}]
 dyn_hair_bk_04 [{shapeType=Box, extentX=6.1017, extentY=12.3439, extentZ=2.6269, center=<4.619 -0.1887 1.2445>}]
 dyn_hair_bk_l_01 [{shapeType=Box, extentX=6.1017, extentY=7.5, extentZ=0.856, center=<3.3857 0 0>}]
 dyn_hair_bk_l_02 [{shapeType=Box, extentX=6.1017, extentY=14.4821, extentZ=0.856, center=<3.3857 0 0>}]
 dyn_hair_bk_l_03 [{shapeType=Box, extentX=8.1017, extentY=6.9821, extentZ=0.856, center=<4.0801 0 -0.2339>}]
 dyn_hair_bk_l_04 [{shapeType=Box, extentX=6.6017, extentY=5.4821, extentZ=0.856, center=<3.5094 0 -0.2061>, rotation=<rot -0.0021 -0.0002 0.001>}]
 dyn_hair_bk_r_01 [{shapeType=Box, extentX=6.1017, extentY=5, extentZ=2.2289, center=<3.3857 0 0>}]
 dyn_hair_bk_r_02 [{shapeType=Box, extentX=6.1017, extentY=5, extentZ=2.2289, center=<3.3291 -0.0261 -0.0768>, rotation=<rot 0 0 18.7845>}]
 dyn_hair_bk_r_03 [{shapeType=Box, extentX=8.1017, extentY=5, extentZ=2.2289, center=<4.4957 -0.009 0>}]
 dyn_hair_bk_r_04 [{shapeType=Box, extentX=6.1017, extentY=5, extentZ=2.2289, center=<3.2204 0.0033 -0.0309>}]
 dyn_hair_bk_r_05 [{shapeType=Box, extentX=6.6017, extentY=5, extentZ=2.7289, center=<3.191 -0.0149 -0.0244>}]
 dyn_hair_fr_01 [{shapeType=Box, extentX=6.1017, extentY=2.1817, extentZ=1.6199, center=<3.4 0 0>}]
 dyn_hair_l_01 [{shapeType=Box, extentX=6.1017, extentY=4.794, extentZ=0.856, center=<3.2306 0 0>}]
 dyn_hair_l_02 [{shapeType=Box, extentX=6.1017, extentY=8.0419, extentZ=0.856, center=<3.3857 0 0>}]
 dyn_hair_l_03 [{shapeType=Box, extentX=6.4216, extentY=7.4734, extentZ=1.856, center=<2.6377 -0.8266 -0.2013>, rotation=<rot 1.4339 32.9812 -2.237>}]
 dyn_hair_l_04 [{shapeType=Box, extentX=9.9216, extentY=1.9734, extentZ=0.856, center=<2.9812 -0.3083 -0.0616>, rotation=<rot -0.0557 -0.0936 79.9978>}]
 dyn_hair_r_01 [{shapeType=Box, extentX=3.9772, extentY=7.5, extentZ=0.856, center=<4.0604 0 0>}]
 dyn_hair_r_02 [{shapeType=Box, extentX=6.1017, extentY=10.2587, extentZ=0.856, center=<3.3857 0 0>}]
 dyn_hair_r_03 [{shapeType=Box, extentX=9.6017, extentY=3.7587, extentZ=0.856, center=<2.9713 -0.4823 -0.8092>, rotation=<rot -14.8889 -20.895 -23.3443>}]
 dyn_hair_r_04 [{shapeType=Box, extentX=9.6017, extentY=2.7587, extentZ=0.856, center=<4.8841 -0.0001 0.001>, rotation=<rot 0.0367 -0.0032 99.9995>}]
 dyn_hair_r_06 [{shapeType=Box, extentX=6.1017, extentY=2.2587, extentZ=2.356, center=<3.3063 0.0026 -0.0193>, rotation=<rot -13.9959 -2.5288 100.3105>}]
 dyn_hair_r_07 [{shapeType=Box, extentX=3.6017, extentY=1.5, extentZ=0.856, center=<2.6606 -0.0001 -0.0004>, rotation=<rot 0.0367 -0.0032 99.9995>}]
 head [
  {shapeType=Capsule, length=5.3867, radius=6.6133, center=<3.053 0.2123 0.1132>, rotation=<rot 90 0 0>}
  {shapeType=Capsule, length=4.8867, radius=7.1133, center=<3.0531 3.1259 0.1132>, rotation=<rot 90 0 0>}
  {shapeType=Capsule, length=3.25, radius=6.6133, center=<2.8368 -3.132 14.8493>, rotation=<rot 83.1573 -96.7375 -21.2703>}
  {shapeType=Capsule, length=5.3867, radius=6.6133, center=<6.1623 -0.8208 -14.699>, rotation=<rot 90 0 0>}
  {shapeType=Capsule, length=10.8867, radius=1.1133, center=<-2.4075 1.2601 -10.4236>, rotation=<rot 70.4842 -74.4124 -0.0001>}
  {shapeType=Capsule, length=2.3867, radius=7.1133, center=<4.2246 3.8132 1.1064>, rotation=<rot 90 0 0>}
 ]
 neck_01 [{shapeType=Capsule, length=8.4045, radius=6.9227, rotation=<rot 90 0 0>}]
 spine_04 [
  {shapeType=Capsule, length=15.755, radius=7.755, center=<13.1 -5.5302 -0.0001>, rotation=<rot 0 -89.9998 179.9999>}
  {shapeType=Capsule, length=15.755, radius=7.755, center=<12.6821 -0.2487 -0.0001>, rotation=<rot 0 -89.9998 179.9999>}
 ]
 spine_05 [
  {shapeType=Capsule, length=0, radius=9, center=<3.7315 -2.7458 0.0088>, rotation=<rot 0 15.0006 0>}
  {shapeType=Capsule, length=15.75, radius=5, center=<5.2684 10.2972 10.2502>, rotation=<rot 0.4352 10.0216 -85.0186>}
  {shapeType=Capsule, length=16, radius=4.25, center=<4.1623 8.5667 -9.9197>, rotation=<rot -33.9014 1.3785 -92.3948>}
 ]
}
constraints [
 {
  bone1Name=head
  bone2Name=neck_01
  swing1LimitDegrees=45
  swing1Motion=Free
  swing2LimitDegrees=45
  swing2Motion=Free
  twistLimitDegrees=45
  twistMotion=Free
 }
 {bone1Name=dyn_hair_bk_r_01, bone2Name=head, swing1LimitDegrees=15, swing2LimitDegrees=15, twistLimitDegrees=1}
 {bone1Name=dyn_hair_bk_r_02, bone2Name=dyn_hair_bk_r_01, swing1LimitDegrees=15, swing2LimitDegrees=15, twistLimitDegrees=1}
 {bone1Name=dyn_hair_bk_r_03, bone2Name=dyn_hair_bk_r_02, swing1LimitDegrees=15, swing2LimitDegrees=15, twistLimitDegrees=1}
 {bone1Name=dyn_hair_bk_l_01, bone2Name=head, swing1LimitDegrees=15, swing2LimitDegrees=15, twistLimitDegrees=1}
 {bone1Name=dyn_hair_bk_l_02, bone2Name=dyn_hair_bk_l_01, swing1LimitDegrees=20, swing2LimitDegrees=20, twistLimitDegrees=1}
 {bone1Name=dyn_hair_bk_l_03, bone2Name=dyn_hair_bk_l_02, swing1LimitDegrees=45, swing2LimitDegrees=45, twistLimitDegrees=1}
 {bone1Name=dyn_hair_bk_01, bone2Name=head, swing1LimitDegrees=45, swing2LimitDegrees=45, twistLimitDegrees=1}
 {bone1Name=dyn_hair_bk_02, bone2Name=dyn_hair_bk_01, swing1LimitDegrees=45, swing2LimitDegrees=45, twistLimitDegrees=1}
 {bone1Name=dyn_hair_bk_03, bone2Name=dyn_hair_bk_02, swing1LimitDegrees=45, swing2LimitDegrees=45, twistLimitDegrees=1}
 {bone1Name=dyn_hair_l_01, bone2Name=head, swing1LimitDegrees=15, swing2LimitDegrees=15, twistLimitDegrees=1}
 {bone1Name=dyn_hair_l_02, bone2Name=dyn_hair_l_01, swing1LimitDegrees=45, swing2LimitDegrees=45, twistLimitDegrees=1}
 {bone1Name=dyn_hair_l_03, bone2Name=dyn_hair_l_02, swing1LimitDegrees=45, swing2LimitDegrees=45, twistLimitDegrees=1}
 {bone1Name=dyn_hair_fr_01, bone2Name=head, swing1LimitDegrees=10, swing2LimitDegrees=10, twistLimitDegrees=1}
 {bone1Name=dyn_hair_r_01, bone2Name=head, swing1LimitDegrees=25, swing2LimitDegrees=10, twistLimitDegrees=1}
 {bone1Name=dyn_hair_r_02, bone2Name=dyn_hair_r_01, swing1LimitDegrees=25, swing2LimitDegrees=10, twistLimitDegrees=1}
 {bone1Name=dyn_hair_r_03, bone2Name=dyn_hair_r_02, swing1LimitDegrees=45, swing2LimitDegrees=45, twistLimitDegrees=1}
 {
  bone1Name=clavicle_l
  bone2Name=spine_05
  swing1LimitDegrees=45
  swing1Motion=Free
  swing2LimitDegrees=45
  swing2Motion=Free
  twistLimitDegrees=45
  twistMotion=Free
 }
 {
  bone1Name=clavicle_r
  bone2Name=spine_05
  swing1LimitDegrees=45
  swing1Motion=Free
  swing2LimitDegrees=45
  swing2Motion=Free
  twistLimitDegrees=45
  twistMotion=Free
 }
 {
  bone1Name=neck_01
  bone2Name=spine_05
  swing1LimitDegrees=45
  swing1Motion=Free
  swing2LimitDegrees=45
  swing2Motion=Free
  twistLimitDegrees=45
  twistMotion=Free
 }
 {bone1Name=dyn_hair_r_04, bone2Name=dyn_hair_r_03, swing1LimitDegrees=25, swing2LimitDegrees=25, twistLimitDegrees=1}
 {bone1Name=dyn_hair_l_04, bone2Name=dyn_hair_l_03, swing1LimitDegrees=25, swing2LimitDegrees=25, twistLimitDegrees=1}
 {bone1Name=dyn_hair_r_06, bone2Name=dyn_hair_r_03, swing1LimitDegrees=25, swing2LimitDegrees=25, twistLimitDegrees=1}
 {bone1Name=dyn_hair_bk_r_04, bone2Name=dyn_hair_bk_r_03, swing1LimitDegrees=15, swing2LimitDegrees=15, twistLimitDegrees=10}
 {bone1Name=dyn_hair_r_07, bone2Name=dyn_hair_r_06, swing1LimitDegrees=25, swing2LimitDegrees=25, twistLimitDegrees=1}
 {bone1Name=dyn_hair_bk_r_05, bone2Name=dyn_hair_bk_r_04, swing1LimitDegrees=15, swing2LimitDegrees=15, twistLimitDegrees=10}
 {bone1Name=dyn_hair_bk_04, bone2Name=dyn_hair_bk_03, swing1LimitDegrees=25, swing2LimitDegrees=25, twistLimitDegrees=1}
 {bone1Name=dyn_hair_bk_l_04, bone2Name=dyn_hair_bk_l_03, swing1LimitDegrees=25, swing2LimitDegrees=25, twistLimitDegrees=1}
 {
  bone1Name=spine_05
  bone2Name=spine_04
  swing1LimitDegrees=45
  swing1Motion=Free
  swing2LimitDegrees=45
  swing2Motion=Free
  twistLimitDegrees=45
  twistMotion=Free
 }
]
