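# Post-process-style animation blueprint for female large Mechstructor face accessories, adding secondary motion to hair, headset microphone and antenna.
# Copied parent pose runs through hair Control Rig, RigidBody, bone constraints and AnimDynamics; targeting offsets the microphone.
# Head-to-neck yaw drives HeadBack, blending a corrective offset on right back hair.
# Skydive-dependent gravity and rigid alpha are set but unread here; several graph nodes sit disconnected from output.
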AnimBlueprint /BRCosmetics/Characters/Player/Female/Large/Bodies/F_LRG_Mechstructor/Meshes/Parts/F_LRG_Mechstructor_FaceAcc_AnimBP
target_skeleton=&/F_LRG_Mechstructor_FaceAcc_Skeleton
graphs [
 AnimGraph {nodes [
   n1 = SelectFloat(A=0.12, B=0.2, bPickA=Get bIsTargeting())
   n2 = Anim Dynamics - Bone: dyn_antenna_01(Alpha=@n1, ExternalForce="0,0,0") {
    angularDampingOverride=0.99
    angularSpringConstant=150
    bAngularSpring=true
    bGravityOverrideInSimSpace=true
    bOverrideAngularDamping=true
    bOverrideLinearDamping=true
    gravityScale=0.5
    lODThreshold=1
    linearDampingOverride=0.99
    numSolverIterationsPostUpdate=4
    numSolverIterationsPreUpdate=16
    alphaBoolBlend {blendOption=Linear}
    boundBone {boneName=dyn_antenna_01}
    componentLinearAccScale=<0.5 0.5 0.5>
    physicsBodyDefinitions [{
      sphereCollisionRadius=0
      boundBone {boneName=dyn_antenna_01}
      boxExtents=<5 5 5>
      constraintSetup {twistAxis=AxisZ, angularLimitsMax=<40 40 40>, angularLimitsMin=<-40 -40 -40>, angularTarget=<0 0 0>}
      localJointOffset=<0 0 -1>
     }]
    relativeSpaceBone {boneName=neck_01}
   }
   Anim Dynamics - Bone: dyn_antenna_02(ComponentPose=@n2, Alpha=@n1, ExternalForce="0,0,0") {
    angularDampingOverride=0.99
    angularSpringConstant=150
    bAngularSpring=true
    bGravityOverrideInSimSpace=true
    bOverrideAngularDamping=true
    bOverrideLinearDamping=true
    gravityScale=0.5
    lODThreshold=1
    linearDampingOverride=0.99
    numSolverIterationsPostUpdate=4
    numSolverIterationsPreUpdate=16
    alphaBoolBlend {blendOption=Linear}
    boundBone {boneName=dyn_antenna_02}
    componentLinearAccScale=<0.5 0.5 0.5>
    physicsBodyDefinitions [{
      sphereCollisionRadius=0
      boundBone {boneName=dyn_antenna_02}
      boxExtents=<5 5 5>
      constraintSetup {twistAxis=AxisZ, angularLimitsMax=<40 40 40>, angularLimitsMin=<-40 -40 -40>, angularTarget=<0 0 0>}
      localJointOffset=<0 0 -1>
     }]
    relativeSpaceBone {boneName=neck_01}
   }
   Blend Poses by bool(
    BlendPose_0=Use cached pose 'AnimCache'()
    BlendPose_1=Component To Local()
    BlendTime_0=0.2
    BlendTime_1=0.2
    bActiveValue=Get bIsTargeting()
   ) {blendType=Linear}
   Component To Local()
   Component To Local()
   Local To Component()
   "Transform (Modify) Bone - Bone: dyn_hair_bk_r_02"(Alpha=1, Translation="-5,-3,4", Rotation="0,0,0", Scale="1,1,1") {
    rotationMode=BMM_Additive
    scaleMode=BMM_Additive
    translationMode=BMM_Additive
    alphaBoolBlend {blendOption=Linear}
    boneToModify {boneName=dyn_hair_bk_r_02}
    translation=<-5 -3 4>
   }
   RigidBody() {
    bClampLinearTranslationLimitToRefPose=true
    bOverrideWorldGravity=true
    overridePhysicsAsset=&/F_LRG_Mechstructor_antenna_Physics
    simulationSpace=BaseBoneSpace
    alphaBoolBlend {blendOption=Linear}
    baseBoneRef {boneName=spine_05}
    componentAppliedLinearAccClamp=<3000 3000 3000>
    componentLinearAccScale=<0.35 0.35 0.35>
    componentLinearVelScale=<0.1 0.1 0.1>
    overrideWorldGravity {z=-980}
   }
   n3 = Control Rig(Source=Copy Pose From Mesh() {bCopyCurves=true, bUseAttachedParent=true}) {
    alphaBoolBlend {blendOption=Linear}
    controlRigAssetReference {blueprintRigClass=&/F_LRG_Mechstructor_Hair_CtrlRig.F_LRG_Mechstructor_Hair_CtrlRig_C}
   }
   n4 = RigidBody(ComponentPose=Local To Component(LocalPose=@n3)) {
    bClampLinearTranslationLimitToRefPose=true
    bOverrideWorldGravity=true
    overridePhysicsAsset=&/F_MED_Mechstructor_Hair_headset_FaceAcc_Physics
    simulationSpace=BaseBoneSpace
    alphaBoolBlend {blendOption=Linear}
    baseBoneRef {boneName=spine_05}
    componentAppliedLinearAccClamp=<3000 3000 3000>
    componentLinearAccScale=<0.35 0.35 0.35>
    componentLinearVelScale=<0.1 0.1 0.1>
    overrideWorldGravity {z=-980}
   }
   n5 = "Constraint - dyn_hair_bk_r_03 (P)"(ComponentPose=@n4, Alpha=0.3, ConstraintWeights_0=1) {
    alphaBoolBlend {blendOption=Linear}
    boneToModify {boneName=dyn_hair_bk_r_03}
    constraintSetup [{targetBone {boneName=dyn_hair_bk_02}}]
    constraintWeights [1]
   }
   n6 = "Constraint - dyn_hair_bk_l_03 (P)"(ComponentPose=@n5, Alpha=0.3, ConstraintWeights_0=1) {
    alphaBoolBlend {blendOption=Linear}
    boneToModify {boneName=dyn_hair_bk_l_03}
    constraintSetup [{targetBone {boneName=dyn_hair_bk_02}}]
    constraintWeights [1]
   }
   n7 = "Transform (Modify) Bone - Bone: dyn_microphone"(ComponentPose=@n6, Alpha=SelectFloat(A=1, B=0, bPickA=Get bIsTargeting()), Translation="0,0,0", Rotation="0,-25,0", Scale="1,1,1") {
    rotationMode=BMM_Additive
    scaleMode=BMM_Additive
    translationMode=BMM_Additive
    alphaBoolBlend {blendOption=Linear}
    boneToModify {boneName=dyn_microphone}
    rotation {yaw=-25}
   }
   n8 = Anim Dynamics - Bone: dyn_microphone(ComponentPose=@n7, Alpha=0.2, ExternalForce="0,0,0") {
    angularDampingOverride=0.98
    angularSpringConstant=300
    bAngularSpring=true
    bGravityOverrideInSimSpace=true
    bOverrideAngularDamping=true
    lODThreshold=1
    linearDampingOverride=0.98
    numSolverIterationsPostUpdate=4
    numSolverIterationsPreUpdate=16
    alphaBoolBlend {blendOption=Linear}
    boundBone {boneName=dyn_microphone}
    componentAppliedLinearAccClamp=<5000 5000 5000>
    physicsBodyDefinitions [{
      sphereCollisionRadius=0
      boundBone {boneName=dyn_microphone}
      boxExtents=<5 5 5>
      constraintSetup {angularTargetAxis=AxisZ, twistAxis=AxisZ, angularLimitsMax=<15 0 0>, angularLimitsMin=<-15 0 0>, angularTarget=<0 0 0>}
      localJointOffset=<0 0 -0.6>
     }]
   }
   n9 = Anim Dynamics - Bone: dyn_antenna_01(ComponentPose=@n8, Alpha=0.2, ExternalForce="0,0,0") {
    angularDampingOverride=0.98
    angularSpringConstant=3200
    bAngularSpring=true
    bGravityOverrideInSimSpace=true
    bOverrideAngularDamping=true
    lODThreshold=1
    linearDampingOverride=0.98
    numSolverIterationsPostUpdate=4
    numSolverIterationsPreUpdate=16
    alphaBoolBlend {blendOption=Linear}
    boundBone {boneName=dyn_antenna_01}
    componentAppliedLinearAccClamp=<5000 5000 5000>
    physicsBodyDefinitions [{
      sphereCollisionRadius=0
      boundBone {boneName=dyn_antenna_01}
      boxExtents=<2 3 5>
      constraintSetup {angularTargetAxis=AxisZ, twistAxis=AxisZ, angularLimitsMax=<15 10 0>, angularLimitsMin=<-15 -10 0>, angularTarget=<0 0 3>}
      localJointOffset=<0 0 1>
     }]
   }
   n10 = "Transform (Modify) Bone - Bone: dyn_hair_bk_r_02"(ComponentPose=@n9, Alpha=Get HeadBack() -> (HeadBack:double), Translation="-5,-3,4", Rotation="0,0,0", Scale="1,1,1") {
    rotationMode=BMM_Additive
    scaleMode=BMM_Additive
    translationMode=BMM_Additive
    alphaBoolBlend {blendOption=Linear}
    boneToModify {boneName=dyn_hair_bk_r_02}
    translation=<-5 -3 4>
   }
   Output Pose(Result=Component To Local(ComponentPose=@n10)) {layerGroup=None}
   Save cached pose 'AnimCache'() {cacheName=AnimCache}
   Use cached pose 'AnimCache'()
   SelectFloat(A=0, B=0, bPickA=Get bIsTargeting())
   SelectFloat(A=0.12, B=0.2, bPickA=false)
  ]}
 EventGraph {nodes [
   Event BlueprintUpdateAnimation()
   > n1 = Sequence()
   > n2 = Branch(Condition=Get bIsSkyDiving())
   > n3 = Branch(Condition=Get bIsParachuteOpened())
   Set gravity_override(execute=[@n2.else, @n3.then], gravity_override="0, 0,-980")
   > Set Rigids Alpha(Rigids Alpha:double=1) -> (Output_Get:double)
   Set gravity_override(execute=@n3.else, gravity_override="0, 0,-300")
   > Set Rigids Alpha(Rigids Alpha=0.5)
   n4 = GetOwnerSkeletalMeshComponent()
   n5 = BreakRotator(InRot=BreakTransform(InTransform=GetDeltaTransformFromRefPose(self=@n4, BoneName=head, BaseName=neck_02)).Rotation)
   Set HeadBack(execute=@n1.then_1, HeadBack:double=float * float(A=MapRangeClamped(Value=@n5.Yaw, InRangeA=-14, InRangeB=-34), B=1)) -> (Output_Get:double)
   TryGetPawnOwner()
   TryGetPawnOwner()
  ]}
]
variables [
 gravity_override {type=Vector}
 HeadBack {type=number}
 Rigids Alpha {type=number}
]
