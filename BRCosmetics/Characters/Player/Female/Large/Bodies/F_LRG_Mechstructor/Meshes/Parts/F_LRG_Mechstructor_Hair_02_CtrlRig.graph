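# Control Rig for the Hair_02 part of the large female Mechstructor character, correcting the back-right hair chain as the head rotates.
# It reads the head's rotation relative to neck_01 as ZYX Euler angles.
# Past a threshold on the negative Z angle, clamped remaps drive Y and Z offsets, added in local space to dyn_hair_bk_r_01.
# The other hair bones are in the hierarchy but are not modified directly.
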ControlRigBlueprint /BRCosmetics/Characters/Player/Female/Large/Bodies/F_LRG_Mechstructor/Meshes/Parts/F_LRG_Mechstructor_Hair_02_CtrlRig
graphs [
 RigVMFunctionLibrary {links [], nodes []}
 RigVMModel {nodes [
   RigUnit_BeginExecution = &RigUnit_BeginExecution()
   n1 = &RigVMFunction_MathQuaternionToEuler(RotationOrder=ZYX, Value=&RigUnit_GetRelativeBoneTransform(Bone=head, Space=neck_01).Transform.Rotation)
   /Script/ControlRig.RigUnit_ModifyBoneTransforms(
    BoneToModify=((Bone="dyn_hair_bk_r_01",Transform=(Rotation=(X=0,Y=0,Z=0,W=1),Translation=(X=0,Y=0,Z=0),Scale3D=(X=1,Y=1,Z=1))))
    Mode=AdditiveLocal
    Weight=1
    WeightMaximum=1
    WeightMinimum=0
    BoneToModify.0.Transform.Rotation=&RigVMFunction_MathQuaternionFromEuler(
     Euler=(X=0,Y=0,Z=0)
     Euler.Y=&RigVMFunction_MathFloatRemap(SourceMaximum=-50, SourceMinimum=-25, TargetMaximum=-40, TargetMinimum=0, bClamp=true, Value=@n1.Result.Z).Result
     Euler.Z=&RigVMFunction_MathFloatRemap(SourceMaximum=-50, SourceMinimum=-25, TargetMaximum=30, TargetMinimum=0, bClamp=true, Value=@n1.Result.Z).Result
    ).Result
    ExecutePin=@RigUnit_BeginExecution.ExecutePin
   ) {is=struct RigVMUnitNode}
  ]}
]
hierarchy [
 root {loc=<0 0 0>, rot=<0 0 0 1>}
 pelvis {parent=root, loc=<0 -0.648 94.7699>, rot=<0.0127 0.707 -0.0127 -0.707>}
 spine_01 {parent=pelvis, loc=<5.0975 -0.6388 0>, rot=<0 0 0.034 -0.9994>}
 spine_02 {parent=spine_01, loc=<7.2518 -0.035 0>, rot=<0 0 0.0238 -0.9997>}
 spine_03 {parent=spine_02, loc=<6.0104 0 0>, rot=<0 0 0.0149 -0.9999>}
 spine_04 {parent=spine_03, loc=<8.1024 0 0>, rot=<0 0 -0.0132 -0.9999>}
 spine_05 {parent=spine_04, loc=<10.0146 0 0>, rot=<0 0 -0.0015 -1>}
 clavicle_l {parent=spine_05, loc=<4.1651 -0.3023 -5.0942>, rot=<0.0465 -0.7289 -0.009 -0.6829>}
 clavicle_r {parent=spine_05, loc=<4.1652 -0.3023 5.0942>, rot=<0.7289 0.0465 -0.6829 0.009>}
 neck_01 {parent=spine_05, loc=<11.3775 -0.6396 0>, rot=<0 0 -0.2003 -0.9797>}
 neck_02 {parent=neck_01, loc=<4.5627 0 0>, rot=<0 0 0.0504 -0.9987>}
 head {parent=neck_02, loc=<5.3914 0 0>, rot=<0 0 0.115 -0.9934>}
 dyn_hair_bk_r_01 {parent=head, loc=<9.5281 -4.7442 3.2485>, rot=<0.175 -0.2454 0.9522 -0.0489>}
 dyn_hair_bk_r_02 {parent=dyn_hair_bk_r_01, loc=<5.936 0.881 -0.3693>, rot=<0 -0.1591 0 -0.9873>}
 dyn_hair_bk_r_03 {parent=dyn_hair_bk_r_02, loc=<5.6814 0.3342 -0.078>, rot=<0.5048 0.0503 -0.0286 -0.8613>}
 dyn_hair_bk_r_04 {parent=dyn_hair_bk_r_03, loc=<5.2657 0 0>, rot=<-0.0068 0.1093 0.0237 -0.9937>}
 dyn_hair_bk_r_05 {parent=dyn_hair_bk_r_04, loc=<3.2057 0 0>, rot=<-0.0521 0.2903 0.0874 -0.9515>}
 dyn_hair_bk_l_01 {parent=head, loc=<8.3061 -6.0932 -4.6288>, rot=<0.0252 -0.9646 0.249 -0.0832>}
 dyn_hair_bk_l_02 {parent=dyn_hair_bk_l_01, loc=<6.0349 0.4336 -0.2367>, rot=<-0.5026 -0.098 0.074 -0.8557>}
 dyn_hair_bk_l_03 {parent=dyn_hair_bk_l_02, loc=<5.7011 0 0>, rot=<-0.0086 0.1554 -0.0235 -0.9875>}
 dyn_hair_bk_l_04 {parent=dyn_hair_bk_l_03, loc=<5.02 0 0>, rot=<0.0107 0.2599 -0.0701 -0.963>}
 dyn_hair_bk_01 {parent=head, loc=<8.1961 -8.0141 0>, rot=<0.0311 -0.7064 0.7069 -0.0167>}
 dyn_hair_bk_02 {parent=dyn_hair_bk_01, loc=<6.4042 0 0>, rot=<0.0003 0.0128 -0.0476 -0.9988>}
 dyn_hair_bk_03 {parent=dyn_hair_bk_02, loc=<6.4897 0 0>, rot=<0.0017 0.0258 0.0098 -0.9996>}
 dyn_hair_bk_04 {parent=dyn_hair_bk_03, loc=<5.034 0 0>, rot=<0.032 0.3213 -0.0418 -0.9455>}
 dyn_hair_l_01 {parent=head, loc=<6.6908 1.1658 -6.4201>, rot=<0.0094 -0.9974 0.0079 -0.0709>}
 dyn_hair_l_02 {parent=dyn_hair_l_01, loc=<5.6562 0 0>, rot=<0.0085 -0.0637 -0.0005 -0.9979>}
 dyn_hair_l_03 {parent=dyn_hair_l_02, loc=<5.2653 0 0>, rot=<0.0271 -0.1843 -0.1548 -0.9702>}
 dyn_hair_l_04 {parent=dyn_hair_l_03, loc=<4.867 0 0>, rot=<0.0552 -0.0121 -0.141 -0.9884>}
 dyn_hair_fr_01 {parent=head, loc=<11.2011 7.2099 -0.0278>, rot=<0.476 0.3487 0.7988 -0.1172>}
 dyn_hair_fr_02 {parent=dyn_hair_fr_01, loc=<6.3667 0 0>, rot=<0 -0.0637 0 -0.998>}
 dyn_hair_r_01 {parent=head, loc=<8.0416 1.6224 6.0165>, rot=<0.2015 -0.0066 0.9794 -0.0103>}
 dyn_hair_r_02 {parent=dyn_hair_r_01, loc=<5.0587 0 0>, rot=<0.0087 -0.1907 -0.018 -0.9814>}
 dyn_hair_r_03 {parent=dyn_hair_r_02, loc=<5.2648 0 0>, rot=<0 0 0 1>}
 dyn_hair_r_04 {parent=dyn_hair_r_03, loc=<4.8889 -1.5063 -1.3813>, rot=<-0.0745 -0.2188 0.2626 -0.9368>}
 dyn_hair_r_06 {parent=dyn_hair_r_03, loc=<2.3411 3.6044 -0.8616>, rot=<0.6719 -0.1913 0.2073 -0.6848>}
 dyn_hair_r_07 {parent=dyn_hair_r_06, loc=<4.4772 -0.0433 0.475>, rot=<0.0055 -0.1248 -0.0126 -0.9921>}
]
