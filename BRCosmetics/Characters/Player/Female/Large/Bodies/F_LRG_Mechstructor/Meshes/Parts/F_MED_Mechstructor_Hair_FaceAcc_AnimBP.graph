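# Post-style Anim Blueprint for the Mechstructor female hair and face accessory, adding secondary motion to the parent mesh's pose.
# Copied pose feeds a hair Control Rig, head-space RigidBody physics, then partial constraints on back strands.
# Head-versus-neck yaw drives HeadBack, blending an additive offset on the right back bone.
# Skydive checks set gravity and rigid-alpha variables, but the AnimGraph uses fixed values instead.
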AnimBlueprint /BRCosmetics/Characters/Player/Female/Large/Bodies/F_LRG_Mechstructor/Meshes/Parts/F_MED_Mechstructor_Hair_FaceAcc_AnimBP
target_skeleton=&/F_MED_Mechstructor_Hair_FaceAcc_Skeleton
graphs [
 AnimGraph {nodes [
   n1 = Control Rig(Source=Copy Pose From Mesh() {bCopyCurves=true, bUseAttachedParent=true}) {
    alphaBoolBlend {blendOption=Linear}
    controlRigAssetReference {blueprintRigClass=&/F_LRG_Mechstructor_Hair_02_CtrlRig.F_LRG_Mechstructor_Hair_02_CtrlRig_C}
   }
   n2 = RigidBody(ComponentPose=Local To Component(LocalPose=@n1)) {
    bClampLinearTranslationLimitToRefPose=true
    bOverrideWorldGravity=true
    overridePhysicsAsset=&/F_MED_Mechstructor_Hair_02_FaceAcc_Physics
    simulationSpace=BaseBoneSpace
    alphaBoolBlend {blendOption=Linear}
    baseBoneRef {boneName=head}
    componentLinearAccScale=<0.4 0.4 0.4>
    componentLinearVelScale=<0.2 0.2 0.2>
    overrideWorldGravity {z=-980}
   }
   n3 = "Constraint - dyn_hair_bk_r_03 (P)"(ComponentPose=@n2, Alpha=0.3, ConstraintWeights_0=1) {
    alphaBoolBlend {blendOption=Linear}
    boneToModify {boneName=dyn_hair_bk_r_03}
    constraintSetup [{targetBone {boneName=dyn_hair_bk_02}}]
    constraintWeights [1]
   }
   n4 = "Constraint - dyn_hair_bk_l_03 (P)"(ComponentPose=@n3, Alpha=0.3, ConstraintWeights_0=1) {
    alphaBoolBlend {blendOption=Linear}
    boneToModify {boneName=dyn_hair_bk_l_03}
    constraintSetup [{targetBone {boneName=dyn_hair_bk_02}}]
    constraintWeights [1]
   }
   n5 = "Transform (Modify) Bone - Bone: dyn_hair_bk_r_02"(ComponentPose=@n4, Alpha=Get HeadBack() -> (HeadBack:double), Translation="-5,-3,4", Rotation="0,0,0", Scale="1,1,1") {
    rotationMode=BMM_Additive
    scaleMode=BMM_Additive
    translationMode=BMM_Additive
    alphaBoolBlend {blendOption=Linear}
    boneToModify {boneName=dyn_hair_bk_r_02}
    translation=<-5 -3 4>
   }
   Output Pose(Result=Component To Local(ComponentPose=@n5)) {layerGroup=None}
  ]}
 EventGraph {nodes [
   Event BlueprintUpdateAnimation()
   > n1 = Sequence()
   > n2 = Branch(Condition=Get bIsSkyDiving())
   > n3 = Branch(Condition=Get bIsParachuteOpened())
   Set gravity_override(execute=[@n2.else, @n3.then], gravity_override="0, 0,-980")
   > Set Rigids Alpha(Rigids Alpha:double=1) -> (Output_Get:double)
   Set gravity_override(execute=@n3.else, gravity_override="0, 0,-300")
   > Set Rigids Alpha(Rigids Alpha=0.5)
   n4 = GetOwnerSkeletalMeshComponent()
   n5 = BreakRotator(InRot=BreakTransform(InTransform=GetDeltaTransformFromRefPose(self=@n4, BoneName=head, BaseName=neck_02)).Rotation)
   Set HeadBack(execute=@n1.then_1, HeadBack:double=float * float(A=MapRangeClamped(Value=@n5.Yaw, InRangeA=-14, InRangeB=-34), B=1)) -> (Output_Get:double)
   TryGetPawnOwner()
   TryGetPawnOwner()
   TryGetPawnOwner()
  ]}
]
variables [
 gravity_override {type=Vector}
 HeadBack {type=number}
 Rigids Alpha {type=number}
]
